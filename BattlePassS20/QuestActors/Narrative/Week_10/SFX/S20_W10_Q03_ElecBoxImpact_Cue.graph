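# Sound cue for an electrical box impact effect in a Season 20 Week 10 narrative quest, routed to the Fort_Quests sound class.
# A Random node picks one of four impact wave variations with equal weight and passes it through a Modulator at default settings.
# It spatializes using a hack attenuation asset borrowed from the Week 7 quest audio, with slightly reduced volume.
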SoundCue /BattlePassS20/QuestActors/Narrative/Week_10/SFX/S20_W10_Q03_ElecBoxImpact_Cue
root=Modulator_0
nodes [
 {id=Modulator_0, children [Random_0]}
 {id=Random_0, children [WavePlayer_0,WavePlayer_1,WavePlayer_2,WavePlayer_3], properties {weights [1,1,1,1]}}
 {id=WavePlayer_0, properties {soundWaveAssetPtr=&/S20_W10_Q03_ElecBoxImpact_01}}
 {id=WavePlayer_1, properties {soundWaveAssetPtr=&/S20_W10_Q03_ElecBoxImpact_02}}
 {id=WavePlayer_2, properties {soundWaveAssetPtr=&/S20_W10_Q03_ElecBoxImpact_03}}
 {id=WavePlayer_3, properties {soundWaveAssetPtr=&/S20_W10_Q03_ElecBoxImpact_04}}
]
properties {
 attenuationSettings=/BattlePassS20/QuestActors/Narrative/Week_07/SFX/S20_W07_Q04_Hack_Atten
 maxDistance=1300
 soundClassObject=&Fort_Quests
 volumeMultiplier=0.95
}
