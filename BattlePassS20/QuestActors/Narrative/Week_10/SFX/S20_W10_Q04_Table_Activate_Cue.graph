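# Sound cue for the table activation sound in a Season 20, Week 10 narrative Battle Pass quest.
# A Delay node at the root holds back a single Wave Player that plays the activation sound wave.
# It is routed to the Fort_Quests sound class and uses a table-specific attenuation asset for distance falloff.
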SoundCue /BattlePassS20/QuestActors/Narrative/Week_10/SFX/S20_W10_Q04_Table_Activate_Cue
root=Delay_0
nodes [
 {id=Delay_0, children [WavePlayer_0], properties {delayMin=1}}
 {id=WavePlayer_0, properties {soundWaveAssetPtr=&/S20_W10_Q04_Table_Activate}}
]
properties {
 attenuationSettings=&/S20_W10_Q04_Table_Atten
 maxDistance=1700
 soundClassObject=&Fort_Quests
 volumeMultiplier=0.98
}
