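# Looping sound cue for a table quest actor in the Season 20 Battle Pass Week 10 narrative quest Q04.
# A looping wave player feeds an Enveloper whose volume curve ramps from silence to full over the first second, fading the loop in.
# The cue uses the Fort_Quests sound class, a dedicated table attenuation asset for distance falloff, and a slightly reduced overall volume.
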SoundCue /BattlePassS20/QuestActors/Narrative/Week_10/SFX/S20_W10_Q04_Table_LP_Cue
root=Enveloper_0
nodes [
 {
  id=Enveloper_0
  children [WavePlayer_0]
  properties {pitchCurve {editorCurveData {keys [{time=0}]}}, volumeCurve {editorCurveData {keys [<at 0 0>,<at 1 1>]}}}
 }
 {id=WavePlayer_0, properties {bLooping=true, soundWaveAssetPtr=&/S20_W10_Q04_Table_LP}}
]
properties {
 attenuationSettings=&/S20_W10_Q04_Table_Atten
 maxDistance=1700
 soundClassObject=&Fort_Quests
 volumeMultiplier=0.85
}
